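; Pickup actor Blueprint for Bling items, adding a spinning mesh and diamond Niagara effects to the standard Fortnite pickup behavior.
; BeginPlay fits the mesh, draw distance and mobile icon, then swaps in a despawn effect five seconds before the curve-table despawn time.
; Toss spawns backpack, BluGlo or Violentium particles or rarity rays, and pickup fades them out.
; Destruction without pickup plays a removal burst and sound, while a timeline drives mobile highlight scaling.
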(event EventBeginPlay
  (bind (_asfort_game_state_athena _bsuccess) (Utilities|Casting|CastToFortGameStateAthena (Game|GetGameState)))
  (bind _skeletal_mesh_pickup (Variables|Default|GetSkeletal_Mesh_Pickup))
  (bind _skeletalorstaticmeshassetprimitive (Variables|Default|GetSkeletalOrStaticMeshAssetPrimitive))
  (bind _itemdefinition (Variables|PickupProperties|GetItemDefinition))
  ;[1 1 1] Begin Play
  (Utilities|FlowControl|Sequence
    (:then_0
      ;; Jon's Parenting Stuff
      (Utilities|Casting|CastToFortPickupBackpack (Actor|GetOwner)
        (:then
          (Variables|Default|SetIs%20a%20Backpack_0 true))
        (:CastFailed
          _))
      (if (Variables|Default|GetIsSkeletalMesh_0)
        (Transformation|SetRelativeLocation _skeletal_mesh_pickup (Variables|Default|GetPivot_Offset_0) false false)
        (Transformation|SetRelativeRotation _skeletal_mesh_pickup (Math|Rotator|SelectRotator "0, 0, 0" (Variables|Default|GetRotation_Bias_0) (Variables|Default|GetIs%20a%20Backpack_0)) false false))
      (Collision|SetCollisionEnabled _skeletalorstaticmeshassetprimitive "NoCollision")
      (Components|Tick|SetComponentTickEnabled _skeletalorstaticmeshassetprimitive false)
      (Transformation|AttachComponentToComponent (Variables|Default|GetSkeletalOrStaticMeshAssetPrimitive) (Variables|Default|GetRoot_0) "None" "KeepRelative" "KeepRelative" "KeepRelative" true)
      (Utilities|IsValid _itemdefinition
        (:Is%20Valid
          (bind _meshtransform (Class|FortItemFunctionLibrary|GetPickupMeshTransform _itemdefinition)
            (:IsIdentity
              (Transformation|SetWorldLocation (Variables|Default|GetSkeletalOrStaticMeshAssetPrimitive) (+ (Transformation|GetActorLocation) (- (Transformation|GetWorldLocation (Variables|Default|GetSkeletalOrStaticMeshAssetPrimitive)) (Collision|GetComponentBounds (Variables|Default|GetSkeletalOrStaticMeshAssetPrimitive)))) false false :!label _label))
            (:IsOverridden
              (Transformation|SetRelativeTransform (Variables|Default|GetSkeletalOrStaticMeshAssetPrimitive) _meshtransform false false)))
          (Utilities|Array|Resize (Variables|Default|GetmidArray_0) (Rendering|Material|GetNumMaterials (Variables|Default|GetSkeletalOrStaticMeshAssetPrimitive))))
        (:Is%20Not%20Valid
          _))
      (goto _label))
    (:then_3
      ;; Set draw distance
      (if (not (Variables|Default|GetIsBluglo_0))
        (bind _output_get (Variables|Default|SetViewDistance_0 (select _bsuccess (select (Fortnite|IsMobilePlatform) (Variables|MaxDrawDistance|GetMax%20Draw%20Distance%20Bacchus_0) (Variables|MaxDrawDistance|GetMax%20Draw%20Distance%20Athena_0)) (Variables|MaxDrawDistance|GetMax%20Draw%20Distance%20StW))))
        (LOD|SetMaxDrawDistance (Variables|Default|GetSkeletalOrStaticMeshAssetPrimitive) (* _output_get (Utilities|Reciprocal (Fortnite|GetNumericCVar "r.ViewDistanceScale"))))
        (bind _backgroundparticlesystem_0 (Variables|Default|GetBackgroundParticleSystem_0)
          (:then
            (LOD|SetMaxDrawDistance _backgroundparticlesystem_0 7000))
          (:else
            _)))
      ;; Randomize Mesh Rotation for BR
      (if (Variables|PickupProperties|GetbRandomRotation)
        (switch Utilities|FlowControl|Switch|SwitchonESubGame (Fortnite|GetSubGame)
          (:Campaign
            _)
          (:Athena
            (bind _output_get_2 (Variables|Default|SetRandom%20Rotation_0 (Math|Random|RandomFloatinRange 0.0 3.6e+02)))
            (Transformation|SetActorRotation 0 (Math|Rotator|MakeRotator 0.0 0.0 _output_get_2) false)
            (Transformation|SetRelativeRotation (Variables|Default|GetBackground) (Math|Rotator|MakeRotator 0.0 0.0 (* _output_get_2 -1.0)) false false))
          (:Invalid
            _)))
      ;; MobileInteractIconSetup
      (Variables|Default|SetMobileInteractionMaterial_0 (Utilities|Select :Index (Variables|Default|GetRarity%20Level_0) :Option%200 "/Game/Effects/Fort_Effects/Gameplay/Pickups/Materials/M_MobileInteractIcon.M_MobileInteractIcon" :Option%201 "/Game/Effects/Fort_Effects/Gameplay/Pickups/Materials/M_MobileInteractIcon_Tier01_Inst.M_MobileInteractIcon_Tier01_Inst" :Option%202 "/Game/Effects/Fort_Effects/Gameplay/Pickups/Materials/M_MobileInteractIcon_Tier02_Inst.M_MobileInteractIcon_Tier02_Inst" :Option%203 "/Game/Effects/Fort_Effects/Gameplay/Pickups/Materials/M_MobileInteractIcon_Tier03_Inst.M_MobileInteractIcon_Tier03_Inst" :Option%204 "/Game/Effects/Fort_Effects/Gameplay/Pickups/Materials/M_MobileInteractIcon_Tier04_Inst.M_MobileInteractIcon_Tier04_Inst")))
    (:then_4
      (Networking|CanExecuteCosmeticEvents
        (:True
          (bind _outxy (DataTable|EvaluateCurveTableRow "/Game/Athena/Balance/DataTables/AthenaGameData.AthenaGameData" "Default.MeteorDrop.Gauntlet.DespawnTime")
            (:RowFound
              (Utilities|FlowControl|Delay (- _outxy 5.0))
              (if (not (Variables|Default|GetPickedUp_0))
                (Niagara|SetNiagaraSystemAsset (Variables|Default|GetNS_Bling_Pickup) "/Game/Effects/Fort_Effects/Effects/Bling/NS_Bling_Diamond_Despawn.NS_Bling_Diamond_Despawn" true)))))))))

(event EventOnPickedUp
  (params (PickupTarget :type Fort%20Pawn%20Object%20Reference))
  (bind _itemdefinition (Variables|PickupProperties|GetItemDefinition))
  ;[1 1 1] item picked up by user
  (Utilities|FlowControl|Sequence
    (:then_0
      (Variables|Default|SetPickedUp_0 true)
      (switch Utilities|FlowControl|Switch|SwitchonESubGame (Fortnite|GetSubGame)
        (:Campaign
          (if (not (Variables|PickupProperties|GetbOwnedByPlayer))
            (Utilities|Casting|CastToFortResourceItemDefinition _itemdefinition
              (:CastFailed
                _)))
          (if (not (>= (Variables|Default|GetRarity%20Level_0) 1))
            (Utilities|Casting|CastToFortIngredientItemDefinition _itemdefinition))
          (Utilities|FlowControl|Delay 0.25))))
    (:then_1
      (CallFunction|SetActiveBackgroundParticleSystem false)
      (CallFunction|SetHiddenBackgroundVisualComponents true)
      (CallFunction|SetVisibleMobileInteractIcon false))
    (:then_2
      ;; Material / FX
      (if (Variables|Default|GetIsLootGiftForOthers_0)
        _
        (elif (== (Variables|Default|GetRarity%20Level_0) 0)
          (CallFunction|SetHologramPickedUpParams true)
          (Utilities|FlowControl|Delay (Variables|HologramSetup|GetPicked%20Up%20Fade%20Out%20Time_0))
          (else
            (switch Utilities|FlowControl|Switch|SwitchonESubGame (Fortnite|GetSubGame)
              (:Campaign
                (CallFunction|SpawnPickedUpTrailPS :!label _label))
              (:Athena
                _)
              (:Invalid
                (goto _label)))
            (Utilities|FlowControl|Delay (Variables|HologramSetup|GetPicked%20Up%20Delay%20Before%20Glow_0))
            (CallFunction|SetHologramPickedUpParams false)
            (Utilities|FlowControl|Delay (Math|Float|float+float (Variables|HologramSetup|GetPicked%20Up%20Fade%20In%20Time_0) (Variables|HologramSetup|GetPicked%20Up%20Hold%20Time_0) (Variables|HologramSetup|GetPicked%20Up%20Fade%20Out%20Time_0) 0.05))))))
    (:then_3
      (bind _backgroundparticlesystem_0 (Variables|Default|GetBackgroundParticleSystem_0)
        (:then
          (Variables|/Script/CascadeRuntime.ParticleSystemComponent|SetCustomTimeDilation (Utilities|Select :Index (Fortnite|GetSubGame) :Campaign 3.0 :Athena 9.0 :Invalid 3.0 :Count 0.0 :Creative 0.0) _backgroundparticlesystem_0)
          (CallFunction|SetupStretchMIDsInternal))
        (:else)))))

(event EventOnAboutToEnterBackpack
  (params (PickupTarget :type Fort%20Pawn%20Object%20Reference))
  ;[1 1 1] End animation and hide asset
  (Utilities|IsValid PickupTarget
    (:Is%20Valid
      (Rendering|SetActorHiddenInGame true)
      ;[1 1 1] For pickups that were on the ground, apply small force feedback as pickup enters backpack
      (if (and (Variables|Default|GetTossed_0) (Pawn|IsLocallyControlled PickupTarget))
        (bind _asfort_player_controller (Utilities|Casting|CastToFortPlayerController (Pawn|GetController PickupTarget)))
        (Game|Feedback|ClientPlayForceFeedback _asfort_player_controller (Variables|Default|GetPickupForceFeedback_0) "Pickup" false false false)))))

(event EventTick
  (params (DeltaSeconds :type Float%20%28single-precision%29))
  ;[1 1 1] Event tick
  (Networking|CanExecuteCosmeticEvents
    (:True
      (Transformation|AddLocalRotation (Variables|Default|GetSkeletalOrStaticMeshAssetPrimitive) (Math|Rotator|MakeRotator 0.0 0.0 (* DeltaSeconds 28.0)) false false))))

(event EventOnTossed
  (bind _backpackparticlesystem (Variables|VFX|GetBackpackParticleSystem))
  (bind _blugloparticlesystem (Variables|VFX|GetBlugloParticleSystem))
  (bind _violentium_idle_fx (Variables|VFX|GetViolentium_Idle_FX))
  ;[0 0 0] Activate particle system if tossed and it should exist
  (Variables|Default|SetTossed_0 true)
  (if (Variables|VFX|GetActivate%20Rarity%20Particle%20Systems_0)
    (if (Variables|Default|GetIs%20a%20Backpack_0)
      ;; Spawn Backpack Particles
      (if (not (Asset|IsAssetNull _backpackparticlesystem))
        (bind _object (Utilities|AsyncLoadAsset _backpackparticlesystem)
          (:Completed
            (bind _asparticle_system (Utilities|Casting|CastToParticleSystem _object))
            (bind _returnvalue (Effects|SpawnEmitterAttached _asparticle_system))
            (bind _output_get (Variables|Default|SetBackgroundParticleSystem_0 _returnvalue))
            (LOD|SetMaxDrawDistance _output_get 5e+03)
            (CallFunction|DestroyBackgroundVisualComponents))))
      (elif (Variables|Default|GetIsBluglo_0)
        ;; Spawn Bluglo Particles
        (if (not (Asset|IsAssetNull _blugloparticlesystem))
          (bind _object_2 (Utilities|AsyncLoadAsset _blugloparticlesystem)
            (:Completed
              (bind _asparticle_system_2 (Utilities|Casting|CastToParticleSystem _object_2))
              (bind _returnvalue_2 (Effects|SpawnEmitterAttached _asparticle_system_2 (Variables|Default|GetRoot_0)))
              (bind _output_get_2 (Variables|Default|SetBackgroundParticleSystem_0 _returnvalue_2))
              (LOD|SetMaxDrawDistance _output_get_2 5e+03)
              (CallFunction|DestroyBackgroundVisualComponents))))
        (elif (Variables|Default|GetIs%20Violentium_0)
          ;; Spawn Violentium Particles
          (if (not (Asset|IsAssetNull _violentium_idle_fx))
            (bind _object_3 (Utilities|AsyncLoadAsset _violentium_idle_fx)
              (:Completed
                (Utilities|Casting|CastToParticleSystem _object_3)
                (bind _returnvalue_3 (Effects|SpawnEmitterAttached 0))
                (bind _output_get_3 (Variables|Default|SetBackgroundParticleSystem_0 _returnvalue_3))
                (LOD|SetMaxDrawDistance _output_get_3 5e+03)
                (CallFunction|DestroyBackgroundVisualComponents))))
          (else
            (bind (_bneedstobeinitialized _returnvalue_4) (FortFX|GetPickupRarityMID (Variables|Default|GetBackground) 0 (Variables|Default|GetRarity%20Level_0)))
            (if _bneedstobeinitialized
              (CallFunction|SetInitialBGRaysParams _returnvalue_4))))))
    (else
      (CallFunction|SetActiveBackgroundParticleSystem false))))

(event EventEnableBacchusHighlight
  (bind _root_0 (Variables|Default|GetRoot_0))
  ;[1 1 1] Mobile highlighting
  (Networking|CanExecuteCosmeticEvents
    (:True
      (bind (_direction _lerpobject) (Timeline|MobileSelectedTL_0 Play)
        (:Update
          (Transformation|SetRelativeScale3D (Variables|Default|GetRoot_0) (Math|Conversions|ToVector(Float) _lerpobject)))
        (:Finished
          (switch Utilities|FlowControl|Switch|SwitchonETimelineDirection _direction
            (:Forward
              (Transformation|SetRelativeScale3D _root_0 "1.000000,1.000000,1.000000"))
            (:Backward
              (Transformation|SetRelativeLocation (Variables|Default|GetRoot_0) "0, 0, 0" false false)
              (Transformation|SetRelativeScale3D _root_0 "1.000000,1.000000,1.000000"))))))))

(event EventDisableBacchusHighlight
  (bind _root_0 (Variables|Default|GetRoot_0))
  ;[1 1 1] Mobile highlighting
  (Networking|CanExecuteCosmeticEvents
    (:True
      (if (not (Variables|Default|GetPickedUp_0))
        (Timeline|MobileSelectedTL_0 Reverse)))))

(event EventEndPlay
  (params (EndPlayReason :type EEndPlayReason%20Enum :default "Destroyed"))
  (bind _skeletalorstaticmeshassetprimitive (Variables|Default|GetSkeletalOrStaticMeshAssetPrimitive))
  (bind _returnvalue (Transformation|GetWorldLocation _skeletalorstaticmeshassetprimitive))
  (Networking|CanExecuteCosmeticEvents
    (:True
      (switch Utilities|FlowControl|Switch|SwitchonEEndPlayReason EndPlayReason
        (:Destroyed
          (if (not (Variables|Default|GetPickedUp_0))
            ;[1 1 1] Despawn fx
            (Niagara|SpawnSystematLocation "/Game/Effects/Fort_Effects/Effects/Bling/NS_Bling_Diamond_Remove.NS_Bling_Diamond_Remove" _returnvalue (Transformation|GetWorldRotation _skeletalorstaticmeshassetprimitive))
            (Audio|PlaySoundatLocation "/Game/Athena/Sounds/LTMs/Bling/Bling_Llama_Disappear_Cue.Bling_Llama_Disappear_Cue" _returnvalue)))))))

(fn SetInitialBGRaysParams
  (params (MID :type Material%20Instance%20Dynamic%20Object%20Reference))
  (bind _viewdistance_0 (Variables|Default|GetViewDistance_0))
  (Rendering|Material|SetVectorParameterValue MID "Color" (Utilities|Array|Get(aref) (Variables|HologramSetup|GetBackgroundRarityColors_0) (Variables|Default|GetRarity%20Level_0)) :!class /Script/Engine.MaterialInstanceDynamic)
  (Rendering|Material|SetScalarParameterValue MID "End Fade" _viewdistance_0 :!class /Script/Engine.MaterialInstanceDynamic)
  (Rendering|Material|SetScalarParameterValue MID "Start Fade" (* _viewdistance_0 0.85) :!class /Script/Engine.MaterialInstanceDynamic)
  (Rendering|Material|SetScalarParameterValue MID "Random" (Math|Random|RandomFloatinRange 0.0 1e+01) :!class /Script/Engine.MaterialInstanceDynamic)
  (if (>= (Variables|Default|GetRarity%20Level_0) 4)
    (Rendering|Material|SetScalarParameterValue MID "IsLegendaryRarity" 1.0 :!class /Script/Engine.MaterialInstanceDynamic)))

(fn ConstructionScript
  (|CallParentFunction)
  (if (not (Variables|FortPickupsParent|GetbIsSkeletalMeshComponent))
    (Transformation|AttachComponentToComponent (Variables|Default|GetNS_Bling_Pickup) (Variables|Default|GetStatic_Mesh_Pickup) "None" "SnapToTarget" "SnapToTarget" "KeepWorld" true))
  (CallFunction|SetupPickupLight (Variables|Default|GetSkeletalOrStaticMeshAssetPrimitive) (Variables|Default|GetSpotLight) (Variables|FortPickupsParent|GetPickupRarityColor))
  ;[1 1 1] Is Blugo
  (if (Utilities|Equal(Object) (Variables|PickupProperties|GetStaticMesh) "/Game/Effects/Fort_Effects/BluGlow/BluGlowModel_NoLOD.BluGlowModel_NoLOD")
    (Variables|Default|SetIsBluglo_0 true)
    (elif (Utilities|Equal(Object) (Variables|PickupProperties|GetStaticMesh) "/Game/Effects/Fort_Effects/BluGlow/Violentium/Violentium_Model_1.Violentium_Model_1")
      ;[1 1 1] is Violentium
      (Variables|Default|SetIs%20Violentium_0 true))))
